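# Translucent, unlit, two-sided Niagara mesh-particle material for the Righteous emote, drawing strokes that wipe on.
# UV1 and particle time build the wipe, remapped by Houdini stroke length stored in vertex red.
# Opacity multiplies the channel-masked stroke texture by the wipe and an alpha-power erosion, with dynamic parameters controlling speed and erosion.
# Emissive tints a foil-mask height blend, driven by two panning noise samples, with particle color.
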Material /Game/Effects/Fort_Effects/Emotes/Righteous/M_Emote_Righteous
graph {
 properties {
  bCastRayTracedShadows=false
  bUsedWithNiagaraMeshParticles=true
  blendMode=BLEND_Translucent
  pixelDepthOffsetMode=PDOM_Legacy
  shadingModel=MSM_Unlit
  twoSided=true
  usedShadingModels=Unlit
 }
 nodes [
  n1 = TextureCoordinate() {coordinateIndex=1, desc=Flrame and Wipe Coordinates}
  n2 = MaterialFunctionCall(Float2=@n1) {materialFunction=&BreakOutFloat2Components}
  n3 = ScalarParameter() {defaultValue=1, parameterName=FlameWidthScalar}
  n4 = VertexColor() {desc="The R channel stores the rationalized \"length\" attribute of each stroke from Houdini."}
  n5 = Divide(B=OneMinus(None=@n4.R))
  n6 = Multiply(A=ParticleRelativeTime(), B=Constant() {r=0.5}) {constB=6}
  n7 = Constant()
  n8 = Constant()
  AppendVector(
   A=Add(A=Multiply(A=@n2.R, B=@n3), B=ScalarParameter() {parameterName=FlameWidthOffset})
   B=Subtract(
    A=Multiply(A=@n2.G, B=@n5)
    B=MaterialFunctionCall(Input=@n6, Input Low=@n7, Input High=@n4.R, Target Low=@n8, Target High=@n5) {desc=Remap from 0 and 1 to -1 and 1, materialFunction=&RemapValueRange}.Result
   )
  )
  DynamicParameter() {
   desc="Wipe: A 0-1 value that wipes the letter on. 0 = Off, 1 = On."
   defaultValue=<rgba 0.9 1 1 1>
   paramNames [Wipe Speed, Alpha Erode, Intensity, WPO Texture Scale]
  }
  DynamicParameter() {
   desc=Speed multiplier on Particle Relative Time
   defaultValue=<rgba 0.9 1 1 1>
   paramNames [Wipe Speed, Alpha Erode, Intensity, WPO Texture Scale]
  }
  n9 = Multiply(A=Time(), B=Constant() {r=0.5})
  n10 = Multiply(A=Constant() {r=-1}, B=@n9)
  n11 = Multiply(A=TextureCoordinate() {vTiling=1.282}, B=Constant() {r=0.7})
  n12 = ComponentMask(None=@n11) {r=true}
  n13 = ComponentMask(None=@n11) {g=true}
  n14 = TextureSample(UVs=AppendVector(A=Add(A=@n10, B=@n12), B=Add(A=@n9, B=@n13))) {desc=Texture A, samplerType=SAMPLERTYPE_Grayscale, texture=&LowResBlurredNoise}
  n15 = Constant() {r=0.8}
  n16 = Constant() {r=0.33}
  n17 = Add(A=@n9, B=ComponentMask(None=@n11) {r=true})
  n18 = ComponentMask(None=@n11) {g=true}
  n19 = Multiply(A=@n9, B=Constant() {r=-1})
  n20 = TextureSample(UVs=AppendVector(A=Add(A=@n16, B=@n17), B=Add(A=@n18, B=@n19))) {desc=Texture B, samplerType=SAMPLERTYPE_Grayscale, texture=&LowResBlurredNoise}
  n21 = TextureSample(UVs=Multiply(A=TextureCoordinate() {vTiling=1.282}, B=Constant() {r=1}) {constB=5}) {samplerType=SAMPLERTYPE_LinearColor, texture=/Game/FrontEnd/Store/Cards/Assets/T_FoilMask}
  n22 = MaterialFunctionCall!(
   Diffuse 1
   Diffuse 2
   Transistion Phase=Multiply(A=Power(Base=@n14.R, Exp=@n15) {constExponent=3}, B=Power(Base=@n20.R, Exp=@n15) {constExponent=3})
   Height Texture 1=Multiply(A=Constant() {r=0.3}, B=@n21.R)
   Height Texture 2=OneMinus(None=@n21.R)
   Contrast
  ) {materialFunction=/Engine/Functions/Engine_MaterialFunctions02/Texturing/HeightLerpWithTwoHeightMaps}
  MaterialFunctionCall(In=@n22.Resulting Height Map, Contrast=ScalarParameter() {defaultValue=0.1, parameterName="Contrast (S)"}) {materialFunction=&CheapContrast}
  MaterialFunctionCall(XY=@n1) {materialFunction=&Swizzle}
  n23 = ScalarParameter() {
   defaultValue=0.1
   desc=A value of 1 makes the flame be the length of the entire strip. Smaller values shorten the length.
   parameterName=FlameLengthScalar
  }
  MaterialFunctionCall(
   Input=Multiply(A=ParticleRelativeTime(), B=Constant() {r=0.5}) {constB=6}
   Target Low=Multiply(A=@n23, B=Constant() {r=1})
   Target High=Multiply(A=@n23, B=Constant() {r=-1})
  ) {materialFunction=&RemapValueRange}
  n24 = ScalarParameter() {
   defaultValue=0.1
   desc=A value of 1 makes the flame be the length of the entire strip. Smaller values shorten the length.
   parameterName=FlameLengthScalar
  }
  Multiply(A=@n24, B=@n4.R) {desc=Scale By stroke length}
  Multiply(
   A=Add(B=ScalarParameter() {defaultValue=0.2, desc=Offset and tweak the flame offset to match the wipe edge., parameterName=FlameLengthOffset})
   B=@n23
  )
  n25 = MaterialFunctionCall(Float2=TextureCoordinate() {coordinateIndex=1}) {materialFunction=&BreakOutFloat2Components}
  NamedRerouteDeclaration(None=@n25.G) {name=WipeVCoord, nodeColor=<rgba 1 0.1875 0 1>}
  n26 = OneMinus(None=@n25.R)
  n27 = DynamicParameter() {
   desc=Speed multiplier on Particle Relative Time
   defaultValue=<rgba 0.9 1 1 1>
   paramNames [Wipe Speed, Alpha Erode, Intensity, WPO Texture Scale]
  }
  n28 = MaterialFunctionCall(
   Input=Multiply(A=ParticleRelativeTime(), B=@n27.Wipe Speed) {constB=6}
   Input Low=Constant()
   Input High=VertexColor() {desc="The R channel stores the rationalized \"length\" attribute of each stroke from Houdini."}.R
   Target Low=Constant() {r=-1}
   Target High=Constant() {r=1}
  ) {desc=Remap from 0 and 1 to -1 and 1, materialFunction=&RemapValueRange}
  NamedRerouteDeclaration(None=Saturate(None=Add(A=@n26, B=@n28.Result))) {name=WipeUCoord, nodeColor=<rgba 0.3125 0 1 1>}
  n29 = TextureObjectParameter() {parameterName=StrokeTexture, samplerType=SAMPLERTYPE_LinearColor, texture=/Game/Effects/Fort_Effects/Emotes/Unruly/T_Emote_Unruly_Ex}
  n30 = TextureSample(UVs=TextureCoordinate() {desc=Stroke Coordinates}, Tex=@n29) {samplerType=SAMPLERTYPE_LinearColor}
  n31 = Multiply(
   A=ChannelMaskParameter(None=@n30.RGB) {parameterName=Mask Stroke}.R
   B=Power(Base=NamedRerouteUsage(), Exp=ScalarParameter() {defaultValue=1, parameterName=Wipe Exponent})
  ) {desc=Mask out stroke with wipe.}
  n32 = DynamicParameter() {
   desc=Speed multiplier on Particle Relative Time
   defaultValue=<rgba 0.9 1 1 1>
   paramNames [Wipe Speed, Alpha Erode, Intensity, WPO Texture Scale]
  }
  n33 = Power(Base=@n30.A, Exp=@n32.Alpha Erode)
  Reroute(None=Saturate(None=Multiply(A=@n31, B=@n33) {desc=Erode})) {desc=OPACITY}
  n34 = ParticleColor()
  n35 = Multiply(A=@n34.RGB) {constB=0.1}
  n36 = Multiply(A=Multiply(A=@n34.RGB, B=LinearInterpolate(Alpha=@n22.Resulting Height Map)), B=Constant() {r=4})
  Reroute(None=MaterialFunctionCall(Emissive=Add(A=@n35, B=@n36)) {materialFunction=&MF_FX_EmissiveAdjustments}.Result) {desc=EMISSIVE}
  VectorParameter() {parameterName=RarityColor, defaultValue=<rgba 0.0062 0.2604 0.0183 1>}
 ]
}
