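# Niagara system for the Righteous emote that emits short-lived, warm HDR GPU sprite sparkles around a paper-sized box.
# Particles spawn continuously on the surface of a box sized by System.PaperSize.
# A narrower, very tall kill box of the same base size culls the middle, leaving sparkles along the outer edges.
# Gravity ramps in over each lifetime alongside curl noise and drag, while alpha and size quickly peak, then fade.
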NiagaraSystem /Game/Effects/Fort_Effects/Emotes/Righteous/NS_Emote_Righteous_PaperSparkles
dependencies {
 usedDataInterfaces [&NiagaraDataInterfaceCurve, &NiagaraDataInterfaceSkeletalMesh]
 usedDynamicInputs []
 usedModules [
  &self:SystemSpawnScript.SystemScriptSource.SystemScriptGraph.NiagaraNodeAssignment_0.SetVariables_#1
  &self:SystemSpawnScript.SystemScriptSource.SystemScriptGraph.NiagaraNodeAssignment_1.SetVariables_#2
  &SystemState
 ]
}
emitters {PaperSparkles {
  simTarget=GPUComputeSim
  emitterUpdateScript [
   &EmitterState(Inactive Response=::NewEnumerator1, Life Cycle Mode=::NewEnumerator1, Loop Duration=1, UseLoopDelay=false)
   &SpawnRate(SpawnRate=100, Use Spawn Probability=false)
  ]
  particleSpawnScript [
   &InitializeParticle(
    Color=<rgba 5 3.7043 1.51 1>
    Lifetime=&Multiply_Float(A=&RandomRangeFloat(Minimum=0.1, Maximum=1), B=1)
    Sprite Size Mode=::NewEnumerator1
    UsePositionOffset=false
    Write Material Random=true
   )
   &ShapeLocation(
    Box Size=$System.PaperSize
    MaskSpawnGroup=false
    Override Seed=false
    Shape Primitive=::NewEnumerator2
    Surface Only=true
    UseOwnerScale=true
   )
   &KillParticlesInVolume(Box Size:Vector3f=&Multiply_Vector(A=$System.PaperSize, B=<0.8 0.8 100>), Kill Shape=::NewEnumerator1)
   &CameraOffset(Camera Offset Amount=50) {enabled=false}
  ]
  particleUpdateScript [
   &ParticleState()
   &GravityForce(Gravity=&Multiply_VectorByFloat(
     Vector=<1 1 -980>
     Float=&FloatFromCurve(FloatCurve=DataInterface() {propertyValues {Curve {keys [<at 0 0 RCIM_Cubic>, {arriveTangent=2.3173, interpMode=RCIM_Cubic, leaveTangent=2.3173, tangentMode=RCTM_User, time=1, value=1}]}}})
    ))
   &CurlNoiseForce(Pan Noise=false)
   &Drag(Use Linear Drag=true, Use Rotational Drag=true)
   &SolveForcesAndVelocity(Clamp Velocity=false, Limit Acceleration=false)
   &ScaleColor(
    Scale Alpha=&FloatFromCurve(FloatCurve=DataInterface() {propertyValues {Curve {keys [<at -0.0092 0.0144>,<at 0.0967 0.9767>,<at 1.0011 0.002>]}}})
    ScaleA=true
    ScaleRGB=true
   )
   &ScaleSpriteSize(Scale Factor=&Vector2DFromFloat(Value=&FloatFromCurve(FloatCurve=DataInterface() {propertyValues {Curve {keys [<at -0.0046 0>,<at 0.1746 0.9952>,<at 1.0011 0.002>]}}})))
  ]
  rendererClasses [&NiagaraSpriteRendererProperties]
  renderers [{rendererClass=&NiagaraSpriteRendererProperties}]
 }}
summary {
 systemName=NS_Emote_Righteous_PaperSparkles
 emitters [{emitterName=PaperSparkles, simTarget=GPUComputeSim, rendererClasses [&NiagaraSpriteRendererProperties]}]
 userVariables []
}
